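# Conversation database for the Squibbly Squid Deep 01 NPC, giving it a stressed restaurant-worker voice when players talk to it.
# One entry point feeds a speech node, which branches through link nodes to the shared quest, basic-service and random-service conversations.
# The speech node has a default line plus contextual lines for first-ever conversation, per-match conversation count, and two player outfits.
# First-meeting and outfit lines take priority from named curve rows.
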ConversationDatabase /NPCLibrary/NPCs/SquibblySquid_Deep/SquibblySquid_Deep_01/NPCLibrary_Conversation_SquibblySquid_Deep_01
entry_points [{tag=Conversation.Entry.SquibblySquidDeep.01, destinations [#1]}]
nodes [
 ConversationEntryPointNode_7 {
  id=#1
  class=ConversationEntryPointNode
  outputs [#2]
  properties {entryTag {tagName=Conversation.Entry.SquibblySquidDeep.01}}
 }
 Speech_C_4 {
  id=#2
  class=Speech_C
  outputs [#3, #4, #5]
  properties {
   nodeName=""
   generalConfig {
    defaultMessage="Don't tell me I missed another order!"
    contextualMessages [
     {
      message="Take a seat! A server will be with you shortly."
      contextRequirements [{requirement {class=FortControllerRequirement_IsFirstEverConversationWithNPC}}]
      priority {curve {rowName=Tandem.Speech.Priority.FirstEverInteraction}}
     }
     {
      message=Make it quick. We're in the weeds.
      contextRequirements [{requirement {
         class=FortControllerRequirement_CheckNPCMatchConversationCount
         properties {nPCUniqueIdTag {tagName=AISpawnerData.Type.Tandem.SquibblySquidDeep.01}}
        }}]
      priority {curve {curveTable=null, rowName=None}}
     }
     {
      message="Don't tell me I missed another order!"
      contextRequirements [{requirement {
         class=FortControllerRequirement_CheckNPCMatchConversationCount
         properties {nPCUniqueIdTag {tagName=AISpawnerData.Type.Tandem.SquibblySquidDeep.01}}
        }}]
      priority {curve {curveTable=null, rowName=None}}
     }
     {
      message=I'm so stressed...
      contextRequirements [{requirement {
         class=FortControllerRequirement_CheckNPCMatchConversationCount
         properties {nPCUniqueIdTag {tagName=AISpawnerData.Type.Tandem.SquibblySquidDeep.01}}
        }}]
      priority {curve {curveTable=null, rowName=None}}
     }
     {
      message="Warm plates and full hearts!"
      contextRequirements [{requirement {class=FortControllerRequirement_HasCID, properties {allowedCIds [/BRCosmetics/Athena/Items/Cosmetics/Characters/Character_SquibblySquid]}}}]
      priority {curve {rowName=Tandem.Speech.Priority.OutfitReaction}}
     }
     {
      message=Ugh... I don't have time to deal with this drama.
      contextRequirements [{requirement {
         class=FortControllerRequirement_HasCID
         properties {allowedCIds [&CID_315_Athena_Commando_M_TeriyakiFish]}
        }}]
      priority {curve {rowName=Tandem.Speech.Priority.OutfitReaction}}
     }
    ]
   }
  }
 }
 ConversationLinkNode_2 {id=#5, class=ConversationLinkNode, properties {remoteEntryTag {tagName=Conversation.Entry.RandomServices}}}
 ConversationLinkNode_0 {id=#4, class=ConversationLinkNode, properties {remoteEntryTag {tagName=Conversation.Entry.BasicServices}}}
 ConversationLinkNode_1 {id=#3, class=ConversationLinkNode, properties {remoteEntryTag {tagName=Conversation.Entry.NPCLibrary.Quests}}}
]
